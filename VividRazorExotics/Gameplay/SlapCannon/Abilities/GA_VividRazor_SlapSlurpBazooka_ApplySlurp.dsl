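; Gameplay ability for the Slap/Slurp Bazooka that applies a Slurp heal to its owner, spilling excess into shield.
; GetMagnitudes caps the scalable heal amount at missing health, then remaining shield room, feeding SetByCaller values on one GameplayEffect.
; Wildlife-tagged pawns receive healing only.
; A next-tick delay works around shields not applying to bots (FORT-549935) and requires ServerOnlyTermination so clients cannot end the ability early.
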(event Ability|EventActivateAbility
  (bind _targetpawn (Variables|Default|GetTargetPawn))
  (bind _returnvalue (FortPawn|GetMaxHealth _targetpawn))
  (bind _returnvalue_2 (FortPawn|GetHealth _targetpawn))
  (bind _rowhealamount (Variables|Default|GetRowHealAmount))
  (bind _returnvalue_3 (ScalableFloat|GetValueatLevel _rowhealamount))
  (bind (_healmagnitude _shieldmagnitude _applyeffect) (CallFunction|GetMagnitudes))
  (bind _returnvalue_4 (- _returnvalue _returnvalue_2))
  (bind _returnvalue_5 (* _returnvalue_4 1e+02))
  (bind _returnvalue_6 (Math|Float|Round _returnvalue_5))
  (bind _returnvalue_7 (Math|Conversions|ToFloat(Integer) _returnvalue_6))
  (bind _returnvalue_8 (* _returnvalue_7 0.01))
  (bind _returnvalue_9 (Math|Float|Clamp(Float) _returnvalue_3 0.0 _returnvalue_8))
  ;[1 0.6756 0.1826] Delay until next tick to avoid a bug where the shield does not apply to bots. See FORT-549935 for more details.
  ; ServerOnlyTermination is also required when using this delay otherwise the client can end the ability before the server has a chance to process
  (Utilities|FlowControl|DelayUntilNextTick)
  (Utilities|IsValid (Variables|Default|GetTargetPawn)
    (:Is%20Valid
      _)
    (:Is%20Not%20Valid
      (bind _asfort_pawn (Utilities|Casting|CastToFortPawn (Ability|GetAvatarActorfromActorInfo))
        (:then
          (Variables|Default|SetTargetPawn _asfort_pawn))
        (:CastFailed
          (Ability|EndAbility)))))
  (if (GameplayTags|HasMatchingGameplayTag (Variables|Default|GetTargetPawn) (Variables|Default|GetWildlife%20Tag))
    ;[0.1129 0.7683 1 0.277] Apply to Wildlife (no shield application)
    (bind _returnvalue_10 (Ability|GameplayEffect|AssignTagSetbyCallerMagnitude (Spec|MakeSpecHandle (Effects|GetDefaultObjectOfGameplayEffectType (Variables|Default|GetGE_HealShield)) (Ability|EffectContext|GetInstigatorActor (Ability|GetGrantedbyEffectContext)) (Variables|Default|GetTargetPawn) 1.0) "(TagName=\"SetByCaller.Healing\")" _returnvalue_9))
    (if (> _returnvalue_9 0.0)
      (Ability|ApplyGameplayEffectSpecToOwner _returnvalue_10)
      (else
        _))
    (Ability|EndAbility)
    (else
      (bind _returnvalue_11 (Ability|GameplayEffect|AssignTagSetbyCallerMagnitude (Spec|MakeSpecHandle (Effects|GetDefaultObjectOfGameplayEffectType (Variables|Default|GetGE_HealShield)) (Ability|EffectContext|GetInstigatorActor (Ability|GetGrantedbyEffectContext)) (Variables|Default|GetTargetPawn) 1.0) "(TagName=\"SetByCaller.Healing\")" _healmagnitude))
      (bind _returnvalue_12 (Ability|GameplayEffect|AssignTagSetbyCallerMagnitude _returnvalue_11 "(TagName=\"SetByCaller.Shield\")" _shieldmagnitude))
      (if _applyeffect
        (Ability|ApplyGameplayEffectSpecToOwner _returnvalue_12)
        (else
          _))
      (Ability|EndAbility))))

(event Ability|EventFailedToActivatePassiveAbility
  (params (ActorInfo :type Gameplay%20Ability%20Actor%20Info%20Structure))
  (|CallParentFunction))

(fn GetMagnitudes
  (return (Math|Float|Clamp(Float) (ScalableFloat|GetValueatLevel (Variables|Default|GetRowHealAmount)) 0.0 (* (Math|Conversions|ToFloat(Integer) (Math|Float|Round (* (- (FortPawn|GetMaxHealth (Variables|Default|GetTargetPawn)) (FortPawn|GetHealth (Variables|Default|GetTargetPawn))) 1e+02))) 0.01)) (Math|Float|Clamp(Float) (Math|Float|Clamp(Float) (- (ScalableFloat|GetValueatLevel (Variables|Default|GetRowHealAmount)) (* (Math|Conversions|ToFloat(Integer) (Math|Float|Round (* (- (FortPawn|GetMaxHealth (Variables|Default|GetTargetPawn)) (FortPawn|GetHealth (Variables|Default|GetTargetPawn))) 1e+02))) 0.01)) 0.0 1e+04) 0.0 (* (Math|Conversions|ToFloat(Integer) (Math|Float|Round (* (- (FortPawn|GetMaxShield (Variables|Default|GetTargetPawn)) (FortPawn|GetShield (Variables|Default|GetTargetPawn))) 1e+02))) 0.01)) (or (> (Math|Float|Clamp(Float) (ScalableFloat|GetValueatLevel (Variables|Default|GetRowHealAmount)) 0.0 (* (Math|Conversions|ToFloat(Integer) (Math|Float|Round (* (- (FortPawn|GetMaxHealth (Variables|Default|GetTargetPawn)) (FortPawn|GetHealth (Variables|Default|GetTargetPawn))) 1e+02))) 0.01)) 0.0) (> (Math|Float|Clamp(Float) (Math|Float|Clamp(Float) (- (ScalableFloat|GetValueatLevel (Variables|Default|GetRowHealAmount)) (* (Math|Conversions|ToFloat(Integer) (Math|Float|Round (* (- (FortPawn|GetMaxHealth (Variables|Default|GetTargetPawn)) (FortPawn|GetHealth (Variables|Default|GetTargetPawn))) 1e+02))) 0.01)) 0.0 1e+04) 0.0 (* (Math|Conversions|ToFloat(Integer) (Math|Float|Round (* (- (FortPawn|GetMaxShield (Variables|Default|GetTargetPawn)) (FortPawn|GetShield (Variables|Default|GetTargetPawn))) 1e+02))) 0.01)) 0.0))))
